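# Physics asset for the Jinn vehicle body cosmetic, providing its simple collision shape.
# A single root body uses one box primitive sized to the car, raised and shifted slightly forward from the root bone.
# No constraints exist, so the body is one rigid piece with no jointed or ragdoll parts.
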PhysicsAsset /VehicleCosmetics/Bodies/Jinn/Meshes/Body_Jinn_PhysicsDM
bodies {root [{shapeType=Box, extentX=444, extentY=216, extentZ=116, shapeName=root_box, center=<24 0 96>}]}
constraints []
